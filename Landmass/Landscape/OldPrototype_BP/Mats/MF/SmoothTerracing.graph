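# Material function that turns a scalar Height into smooth terraces spaced by Interval, in the Landmass landscape prototype folder.
# Each step is an S-curve built by mirroring a power curve about the step's midpoint, with exponent 1/Smoothness, clamped against division by zero.
# A Custom HLSL node computes the output.
# An equivalent node-graph version of the same math remains unconnected to the output.
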MaterialFunction /Landmass/Landscape/OldPrototype_BP/Mats/MF/SmoothTerracing
graph {nodes [
  MaterialExpressionFunctionInput_0 = FunctionInput() {inputName=Height, inputType=FunctionInput_Scalar}
  MaterialExpressionFunctionInput_2 = FunctionInput() {bUsePreviewValueAsDefault=true, inputName=Interval, inputType=FunctionInput_Scalar, sortPriority=1, previewValue=<0.5 0 0 1>}
  n1 = Divide(A=@MaterialExpressionFunctionInput_0, B=@MaterialExpressionFunctionInput_2)
  n2 = Frac(None=@n1)
  n3 = Saturate(None=Ceil(None=Subtract(A=@n2) {constB=0.5}))
  n4 = LinearInterpolate(A=@n2, B=OneMinus(None=@n2), Alpha=@n3)
  MaterialExpressionFunctionInput_1 = FunctionInput() {bUsePreviewValueAsDefault=true, inputName=Smoothness, inputType=FunctionInput_Scalar, sortPriority=2}
  n5 = Max(A=@MaterialExpressionFunctionInput_1, B=Constant() {r=0.0001})
  n6 = Multiply(A=Power(Base=Multiply(A=@n4) {constB=2}, Exp=Divide(B=@n5))) {constB=0.5}
  n7 = OneMinus(None=@n6)
  Add(
   A=Multiply(A=Floor(None=@n1), B=@MaterialExpressionFunctionInput_2)
   B=Multiply(A=LinearInterpolate(A=@n6, B=@n7, Alpha=@n3), B=@MaterialExpressionFunctionInput_2)
  )
  n8 = Custom!(InHeight=@MaterialExpressionFunctionInput_0, Interval=@MaterialExpressionFunctionInput_2, Smoothness=@MaterialExpressionFunctionInput_1) {
   code="Smoothness = 1 / max(Smoothness, 0.0001);\r\n\r\nfloat curlevel = InHeight / Interval;\r\nfloat curphase = frac(curlevel);\r\nfloat halfmask = saturate(ceil(curphase - 0.5));\r\ncurlevel = floor(curlevel) * Interval;\r\n\r\n\r\nfloat scurve = lerp(curphase, 1-curphase, halfmask) * 2;\r\nscurve = pow(scurve, Smoothness) / 2;\r\nscurve = lerp(scurve, 1-scurve, halfmask) * Interval;\r\nreturn scurve + curlevel;\r\n"
   description=Smooth Terracing
   outputType=CMOT_Float1
   inputs [
    {inputName=InHeight, input {expression=&self:MaterialExpressionFunctionInput_0}}
    {inputName=Interval, input {expression=&self:MaterialExpressionFunctionInput_2}}
    {inputName=Smoothness, input {expression=&self:MaterialExpressionFunctionInput_1}}
   ]
  }
  FunctionOutput(None=@n8)
 ]}
