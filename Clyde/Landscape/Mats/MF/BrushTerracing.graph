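# Landscape brush material function that terraces an input height within a distance-field mask.
# It blends Height toward SmoothTerracing's result using a saturated weight from the mask plus TerraceOffset, divided by mask length and scaled by TerraceParams alpha.
# TerraceParams' blue (spacing) channel feeds both Interval and Smoothness; green, labelled smoothness, never reaches the output.
# An alternative two-interval SmoothCeil terrace (golden-ratio spacing) and a noise offset remain unconnected to the output.
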MaterialFunction /Clyde/Landscape/Mats/MF/BrushTerracing
graph {nodes [
  n1 = VectorParameter() {desc="R=Alpha, G=Smoothness,B=Spacing,A=Mask len", parameterName=TerraceParams, defaultValue=<rgba 0 0 1024 1024>}
  n2 = FunctionInput() {bUsePreviewValueAsDefault=true, inputName=DF Mask, inputType=FunctionInput_Scalar, sortPriority=1, previewValue=<1 0 0 1>}
  n3 = ScalarParameter() {parameterName=TerraceOffset}
  n4 = Multiply(A=@n1.R, B=Divide(A=Add(A=@n2, B=@n3), B=@n1.A))
  Abs(None=@n4)
  n5 = FunctionInput() {bUsePreviewValueAsDefault=true, inputName=Height, inputType=FunctionInput_Scalar}
  n6 = Saturate(None=@n4)
  FunctionOutput(None=LinearInterpolate(
    A=@n5
    B=MaterialFunctionCall!(Height=@n5, Interval=@n1.B, Smoothness=@n1.B) {materialFunction=/Landmass/Landscape/OldPrototype_BP/Mats/MF/SmoothTerracing}.Result
    Alpha=@n6
   ))
  n7 = Constant()
  n8 = Add(A=@n5, B=@n7)
  n9 = MaterialFunctionCall!(In=Divide(A=@n8, B=@n1.B), Ramp Width=@n1.G) {materialFunction=/Engine/Functions/Engine_MaterialFunctions01/Math/SmoothCeil}
  n10 = Subtract(A=Multiply(A=Subtract(A=@n9.Result), B=@n1.B), B=@n7) {constB=5000}
  n11 = Divide(A=@n1.B) {constB=1.618}
  n12 = MaterialFunctionCall(In=Divide(A=@n8, B=@n11), Ramp Width=@n1.G) {materialFunction=&SmoothCeil}
  n13 = Subtract(A=Multiply(A=Subtract(A=@n12.Result), B=@n11), B=@n7)
  LinearInterpolate(A=@n5, B=Max(A=@n10, B=@n13), Alpha=@n6)
  Min(A=@n10, B=@n13)
  n14 = Multiply(A=TextureCoordinate(), B=ScalarParameter() {defaultValue=2, parameterName=T})
  n15 = Constant()
  Multiply(
   A=Noise!(World Position=AppendVector(A=@n14, B=@n15), FilterWidth) {levels=2, noiseFunction=NOISEFUNCTION_GradientALU, outputMin=0}
   B=ScalarParameter() {defaultValue=25000, parameterName=TerraceOffsetting}
  )
 ]}
