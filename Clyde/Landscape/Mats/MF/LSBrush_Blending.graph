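# Landscape brush material function that blends a Landmass-style brush shape into the terrain height render target, outputting Height, Layer Alpha and a distance-field Debug value.
# Cached distance-field height, optionally using mesh depth, gains masked displacement, then edge falloff by width or angle shapes it.
# Smooth min/max intersections with existing terrain feed manual blend modes, then terracing.
# Layer alpha derives separately from the alpha texture input; several leftover nodes are unconnected.
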MaterialFunction /Clyde/Landscape/Mats/MF/LSBrush_Blending
graph {nodes [
  n1 = MaterialFunctionCall() {materialFunction=/Landmass/Landscape/OldPrototype_BP/Mats/MF/ReadCachedDFHeight}
  n2 = LinearInterpolate(A=ScalarParameter() {defaultValue=1394, parameterName=Z}, B=@n1.Height, Alpha=ScalarParameter() {parameterName=UseMeshDepth})
  n3 = Constant()
  n4 = ScalarParameter() {defaultValue=0.5, parameterName=Displacement Midpoint}
  n5 = Subtract(A=FunctionInput(Preview=@n3) {bUsePreviewValueAsDefault=true, inputName=Displacement, inputType=FunctionInput_Scalar}, B=@n4)
  n6 = ScalarParameter() {defaultValue=256, parameterName=BrushTexHeight}
  n7 = Add(
   A=@n2
   B=Multiply(
    A=MaterialFunctionCall!(UV Distance=@n1.UV Distance) {materialFunction=/Landmass/Landscape/OldPrototype_BP/Mats/MF/BrushDisplaceMask}.Mask
    B=Multiply(A=@n5, B=@n6)
   ) {desc=Displacement * Mask}
  ) {desc=Height + Displacement}
  n8 = TextureSampleParameter2D() {parameterName=HeightRT, texture=/Engine/EngineResources/Black_Low.Black}
  n9 = MaterialFunctionCall!(AlphaBlend=@n8.R, Min=@n8.R, Max=@n8.R, Additive=Constant()) {materialFunction=/Landmass/Landscape/OldPrototype_BP/Mats/MF/BrushBlendModes_Manual}
  n10 = Multiply(A=@n9.Result)
  n11 = Add(
   A=Multiply(A=@n1.UV Distance, B=ScalarParameter() {defaultValue=1, desc=Tangent, parameterName=FalloffTangent})
   B=ScalarParameter() {parameterName=EdgeCenterOffset}
  )
  n12 = LinearInterpolate(
   A=@n11
   B=MaterialFunctionCall!(A=@n11, B=Constant(), K=VectorParameter() {parameterName=SmoothingParams, defaultValue=<rgba 1 1 1 1>}.R) {materialFunction=/Landmass/Landscape/OldPrototype_BP/Mats/MF/SmoothIntersections}.Smooth Min
   Alpha=ScalarParameter() {parameterName=UseMeshDepth}
  )
  n13 = Multiply(A=Divide(A=@n12, B=ScalarParameter() {defaultValue=1, parameterName=EdgeWidth})) {constB=-1}
  n14 = LinearInterpolate(A=@n7, B=@n10, Alpha=@n13)
  n15 = LinearInterpolate(A=@n7, B=@n10, Alpha=Min(A=@n13))
  n16 = ScalarParameter() {parameterName=UseAngle}
  n17 = LinearInterpolate(A=Max(A=@n14, B=@n15), B=Add(A=Multiply(A=@n12) {constB=-1}, B=@n7) {desc=Negative Shape}, Alpha=@n16)
  n18 = VectorParameter() {parameterName=SmoothingParams, defaultValue=<rgba 1 1 1 1>}
  n19 = Min(A=@n14, B=@n15)
  n20 = Add(A=@n12, B=@n7) {desc=Positive Shape}
  n21 = LinearInterpolate(A=LinearInterpolate(A=@n19, B=Min(A=@n19, B=@n7), Alpha=ScalarParameter() {parameterName=UseMeshDepth}), B=@n20, Alpha=@n16)
  n22 = Subtract(A=@n7, B=@n10) {desc=Intersection of Terrain and Brush}
  n23 = Min(A=Max(A=@n22) {constB=0}, B=@n18.G)
  n24 = ScalarParameter() {parameterName=UseMeshDepth}
  n25 = MaterialFunctionCall(
   A=MaterialFunctionCall(A=@n9.Result, B=@n17, K=@n18.G) {materialFunction=&SmoothIntersections}.Smooth Min
   B=@n21
   K=LinearInterpolate(A=@n18.G, B=@n23, Alpha=@n24)
  ) {materialFunction=&SmoothIntersections}
  n26 = MaterialFunctionCall(A=@n9.Result, B=@n17, K=@n18.G) {materialFunction=&SmoothIntersections}
  n27 = MaterialFunctionCall(A=@n9.Result, B=@n21, K=@n18.G) {materialFunction=&SmoothIntersections}
  n28 = TextureSampleParameter2D() {parameterName=HeightRT, texture=&Black}
  n29 = Add(A=@n25.Smooth Max, B=@n28.R)
  FunctionOutput(None=MaterialFunctionCall!(
    Height=MaterialFunctionCall(AlphaBlend=@n25.Smooth Max, Min=@n26.Smooth Min, Max=@n27.Smooth Max, Additive=@n29) {materialFunction=&BrushBlendModes_Manual}.Result
    DF Mask=@n28.R
   ) {materialFunction=/Landmass/Landscape/OldPrototype_BP/Mats/MF/BrushTerracing}.Result) {outputName=Height}
  n30 = Subtract(
   A=FunctionInput(Preview=@n3) {bUsePreviewValueAsDefault=true, inputName=Alpha Texture, inputType=FunctionInput_Scalar, sortPriority=1}
   B=@n4
  )
  n31 = ScalarParameter() {defaultValue=1, parameterName=WeightmapInfluence}
  FunctionOutput(None=MaterialFunctionCall!(UV Distance=@n1.UV Distance, Displacement=Multiply(A=@n30, B=@n31), Z Height=@n2) {materialFunction=/Landmass/Landscape/OldPrototype_BP/Mats/MF/BrushAlphaMask}.Alpha) {outputName=Layer Alpha, sortPriority=1}
  n32 = MaterialFunctionCall!(UV Offset) {materialFunction=/Landmass/Landscape/OldPrototype_BP/Mats/MF/LSBrush_DistanceField}
  FunctionOutput(None=Abs(None=Subtract(A=@n32.World Distance))) {outputName=Debug, sortPriority=10}
  n33 = ScalarParameter() {defaultValue=1, parameterName=UseCachedDF}
  LinearInterpolate(A=@n32.World Distance, Alpha=@n33)
  n34 = TextureSampleParameter2D() {parameterName=MeshDepth, texture=&Black}
  LinearInterpolate(A=LinearInterpolate(A=@n32.Z Height, B=Subtract(B=@n34.R) {constA=50000}, Alpha=@n32.Z Height), Alpha=@n33)
  n35 = Min(A=Max(A=Multiply(A=@n22) {constB=-1}) {constB=0}, B=@n18.G)
  LinearInterpolate(A=@n18.G, B=@n35, Alpha=@n24)
  OneMinus(None=@n34.R)
  Reroute(None=@n23)
  Reroute(None=@n20)
  Reroute(None=@n35)
  Reroute(None=Max(A=@n23, B=@n18.R))
 ]}
